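# Translucent unlit UI material, named for the HomeBase Power font, that produces a pulsing cyan glow color.
# Two clamped sines of time, offset from each other, are summed to blend cyan toward white over time.
# The V texture coordinate forms a mirrored vertical gradient through SmoothStep, brightest at the middle, blending cyan to pale cyan.
# The pulse and gradient colors are multiplied together, with no texture sampled.
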Material /FortUILibrary/Materials/Foundation/M-HomeBasePower-Font
graph {
 properties {
  bUsedWithStaticMesh=false
  blendMode=BLEND_Translucent
  materialDomain=MD_UI
  pixelDepthOffsetMode=PDOM_Legacy
  shadingModel=MSM_Unlit
  usedShadingModels=Unlit
 }
 nodes [
  n1 = Constant3Vector() {constant=<rgba 0.0953 1 1 1>}
  n2 = Time()
  n3 = Constant() {r=4}
  n4 = Clamp(None=Sine(None=Divide(A=@n2, B=@n3)))
  n5 = Subtract(A=@n2, B=Constant() {r=2})
  n6 = Clamp(None=Sine(None=Divide(A=@n5, B=@n3)))
  n7 = ComponentMask(None=TextureCoordinate()) {g=true}
  n8 = Constant() {r=0.5}
  n9 = Subtract(A=@n7, B=@n8)
  n10 = Subtract(A=OneMinus(None=@n7), B=@n8)
  n11 = OneMinus(None=Add(A=Clamp(None=@n9), B=Clamp(None=@n10)))
  n12 = Constant() {r=0.5}
  n13 = Constant() {r=1}
  Multiply(
   A=LinearInterpolate(A=@n1, B=Constant() {r=1}, Alpha=Add(A=@n4, B=@n6))
   B=LinearInterpolate(
    A=@n1
    B=Constant3Vector() {constant=<rgba 0.7835 1 1 1>}
    Alpha=MaterialFunctionCall(Alpha=@n11, Min=@n12, Max=@n13) {materialFunction=&SmoothStep}.Result
   )
  )
 ]
}
