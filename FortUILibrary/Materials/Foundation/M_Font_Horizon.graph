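# M_Font_Horizon is a UI-domain material that, going by its name, gives fonts a horizon-style color gradient.
# It multiplies the V channels of UVs rotated plus and minus 30 degrees, using constant-time Rotators, to form a gradient mask.
# SmoothStep ramps blend RGB_1 toward RGB_Highlight and RGB_2, with a narrow step switching between the blends to create the horizon edge.
# The s_WhiteFill scalar is added on top to brighten the result toward white.
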Material /FortUILibrary/Materials/Foundation/M_Font_Horizon
graph {
 properties {bUsedWithStaticMesh=false, materialDomain=MD_UI, pixelDepthOffsetMode=PDOM_Legacy, usedShadingModels=Default Lit}
 nodes [
  n1 = VectorParameter() {parameterName=RGB_1, defaultValue=<rgba 0 1 0.9647 1>}
  n2 = VectorParameter() {parameterName=RGB_Highlight, defaultValue=<rgba 1 1 1 1>}
  n3 = TextureCoordinate()
  n4 = Divide(A=Constant() {r=-30}) {constA=360, constB=360}
  n5 = Divide(A=Constant() {r=30}) {constA=360, constB=360}
  n6 = DotProduct(
   A=ComponentMask(None=Rotator(Coordinate=@n3, Time=@n4) {speed=6.2832}) {g=true}
   B=ComponentMask(None=Rotator(Coordinate=@n3, Time=@n5) {speed=6.2832}) {g=true}
  )
  n7 = MaterialFunctionCall(Alpha=@n6, Min=Constant(), Max=Constant() {r=0.3}) {materialFunction=&SmoothStep}
  n8 = VectorParameter() {parameterName=RGB_2, defaultValue=<rgba 0 0.4287 1 1>}
  n9 = MaterialFunctionCall(Alpha=@n6, Min=Constant() {r=0.15}, Max=Constant() {r=0.7}) {materialFunction=&SmoothStep}
  n10 = Constant() {r=0.24}
  n11 = Constant() {r=0.26}
  Add(
   A=LinearInterpolate(
    A=LinearInterpolate(A=@n1.RGB, B=@n2.RGB, Alpha=@n7.Result)
    B=LinearInterpolate(A=@n1.RGB, B=@n8.RGB, Alpha=@n9.Result)
    Alpha=MaterialFunctionCall(Alpha=@n6, Min=@n10, Max=@n11) {materialFunction=&SmoothStep}.Result
   ) {constB=0}
   B=ScalarParameter() {parameterName=s_WhiteFill, sliderMax=1}
  )
 ]
}
