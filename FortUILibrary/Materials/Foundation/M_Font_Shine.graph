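# M_Font_Shine is a UI material, named for fonts, that adds a periodic animated shine over a base color.
# Rotated UVs scroll with time into a sine and SmoothStep, forming bands tinted by the shine color's RGB.
# A time-modulo gate switches the shine on and off in cycles, with durations divided by shine speed.
# The PrestigeActive scalar blends between plain base color and base color plus gated shine.
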Material /FortUILibrary/Materials/Foundation/M_Font_Shine
graph {
 properties {bUsedWithStaticMesh=false, materialDomain=MD_UI, pixelDepthOffsetMode=PDOM_Legacy, usedShadingModels=Default Lit}
 nodes [
  n1 = VectorParameter() {parameterName=BaseColor, defaultValue=<rgba 1 0.5583 0 1>}
  n2 = TextureCoordinate()
  n3 = ScalarParameter() {defaultValue=0.15, group=Shine, parameterName=Shine Rotation}
  n4 = ScalarParameter() {defaultValue=4, group=Shine, parameterName=Shine Speed}
  n5 = OneMinus(None=Multiply(A=Time(), B=@n4))
  n6 = Constant()
  n7 = ComponentMask(None=Add(A=Rotator(Coordinate=@n2, Time=@n3) {centerX=0, centerY=0}, B=AppendVector(A=@n5, B=@n6))) {r=true}
  n8 = ScalarParameter() {defaultValue=0.75, group=Shine, parameterName=Shine SS Min}
  n9 = ScalarParameter() {defaultValue=0.25, group=Shine, parameterName=Shine SS Max}
  n10 = LinearInterpolate(
   A=Constant()
   B=VectorParameter() {group=Shine, parameterName=ShineRGBA, defaultValue=<rgba 1 1 1 0.5>}.RGB
   Alpha=MaterialFunctionCall(Alpha=Sine(None=@n7) {period=2}, Min=@n8, Max=Add(A=@n8, B=@n9)) {materialFunction=&SmoothStep}.Result
  )
  n11 = Constant() {r=1}
  n12 = Time()
  n13 = Divide(A=ScalarParameter() {defaultValue=3, group=Shine, parameterName=Cycles On}, B=@n4)
  n14 = Constant() {r=2}
  n15 = Multiply(A=Divide(A=ScalarParameter() {defaultValue=6, group=Shine, parameterName=Cycles Off}, B=@n4), B=Constant() {r=2})
  n16 = Fmod(A=@n12, B=Add(A=Multiply(A=@n13, B=@n14), B=@n15))
  n17 = Constant()
  n18 = Constant() {r=-1}
  n19 = If(
   A=Add(A=Fmod(A=@n12, B=@n11), B=If(A=@n16, B=@n15, "A > B"=@n17, "A == B"=@n18, "A < B"=@n18)) {constB=-1}
   B=Constant()
   "A > B"=Constant() {r=1}
   "A < B"=Constant()
  )
  LinearInterpolate(
   A=LinearInterpolate(A=@n1.RGB, B=@n1.RGB)
   B=Add(A=Multiply(A=@n10, B=@n19), B=@n1.RGB)
   Alpha=ScalarParameter() {parameterName=PrestigeActive}
  )
 ]
}
