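# M_Font_SimplePulse is a UI-domain material, named for font use, that pulses a color between two RGB vector parameters, red and blue by default.
# Time scaled by the TimeMultiplier parameter drives a sine wave, which is offset, halved and clamped into a blend alpha.
# That alpha interpolates between RGB1 and RGB2, so both colors and the pulse speed are adjustable through parameters.
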Material /FortUILibrary/Materials/Foundation/M_Font_SimplePulse
graph {
 properties {bUsedWithStaticMesh=false, materialDomain=MD_UI, pixelDepthOffsetMode=PDOM_Legacy, usedShadingModels=Default Lit}
 nodes [
  n1 = Time()
  n2 = ScalarParameter() {defaultValue=0.25, parameterName=TimeMultiplier}
  n3 = Add(A=Sine(None=Multiply(A=@n1, B=@n2)))
  LinearInterpolate(
   A=VectorParameter() {parameterName=RGB1, defaultValue=<rgba 1 0 0 1>}.RGB
   B=VectorParameter() {parameterName=RGB2, defaultValue=<rgba 0 0 1 1>}.RGB
   Alpha=Clamp(None=Multiply(A=@n3) {constB=0.5})
  )
 ]
}
