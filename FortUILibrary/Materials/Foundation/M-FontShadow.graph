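# M-FontShadow is a translucent UI material that draws a soft, tintable radial gradient, named for use as a font shadow.
# A RadialGradientExponential on UV channel 1, shaped by Radius and Density parameters, is raised to PowerExponent and scaled by Alpha to form the falloff mask.
# A separate branch multiplies the Color vector parameter by ColorMultiply to set the tint, with no texture sampling in the graph.
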Material /FortUILibrary/Materials/Foundation/M-FontShadow
graph {
 properties {
  bUsedWithStaticMesh=false
  blendMode=BLEND_Translucent
  materialDomain=MD_UI
  pixelDepthOffsetMode=PDOM_Legacy
  usedShadingModels=Default Lit
 }
 nodes [
  Multiply(
   A=ScalarParameter() {defaultValue=1, parameterName=ColorMultiply}
   B=VectorParameter() {parameterName=Color, defaultValue=<rgba 1 1 1 1>}.RGB
  )
  n1 = TextureCoordinate() {coordinateIndex=1}
  n2 = ScalarParameter() {defaultValue=1, parameterName=Radius, sliderMax=1}
  n3 = ScalarParameter() {defaultValue=0.75, parameterName=Density}
  Multiply(
   A=Power(
    Base=MaterialFunctionCall(UVs=@n1, Radius=@n2, Density=@n3) {materialFunction=&RadialGradientExponential}.RadialGradientExponential
    Exp=ScalarParameter() {defaultValue=1.25, parameterName=PowerExponent}
   )
   B=ScalarParameter() {defaultValue=1, parameterName=Alpha, sliderMax=0.5}
  )
 ]
}
