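# Translucent UI material for a pulsing outer glow, named for the HomeBase power display, with a constant cyan-blue colour.
# A 2D radial gradient is squared and smoothstepped to form a soft circular falloff.
# Two time-driven sine waves, offset by half a period and clamped, are summed into a repeating pulse.
# That pulse is scaled down, lifted by a base level so the glow never vanishes, and multiplied into the falloff.
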Material /FortUILibrary/Materials/Foundation/M-HomeBasePower-OuterGlow
graph {
 properties {
  bUsedWithStaticMesh=false
  blendMode=BLEND_Translucent
  materialDomain=MD_UI
  pixelDepthOffsetMode=PDOM_Legacy
  shadingModel=MSM_Unlit
  usedShadingModels=Unlit
 }
 nodes [
  Constant3Vector() {constant=<rgba 0 0.7231 1 1>}
  n1 = MaterialFunctionCall!(UVs, CenterPosition, Radius=Constant() {r=0.49}) {materialFunction=/Engine/Functions/Engine_MaterialFunctions01/Gradient/SphereGradient-2D}
  n2 = Time()
  n3 = Constant() {r=4}
  n4 = Divide(A=@n2, B=@n3)
  n5 = Divide(A=Subtract(A=@n2, B=Constant() {r=2}), B=@n3)
  n6 = Add(A=Clamp(None=Sine(None=@n4)), B=Clamp(None=Sine(None=@n5)))
  n7 = Constant() {r=3}
  Multiply(
   A=MaterialFunctionCall(Alpha=Multiply(A=@n1.Result 0-1, B=@n1.Result 0-1), Min=Constant(), Max=Constant() {r=1.5}) {materialFunction=&SmoothStep}.Result
   B=Add(A=Divide(A=@n6, B=@n7), B=Constant() {r=0.25})
  )
 ]
}
